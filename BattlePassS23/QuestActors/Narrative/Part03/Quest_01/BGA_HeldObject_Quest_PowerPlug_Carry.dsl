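; Carryable power plug for a Season 23 narrative quest that its owning player carries to a socket.
; BeginPlay applies a carry effect and starts pickup; a client timer holds the attach offset, and only the owner can re-interact.
; Held's RepNotify runs socket overlap checks that report the quest stat, drop and hide the plug.
; Gameplay events signal shown, picked-up, dropped and turned-in states; some names remain from sign and bear quests.
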(event OnHeldObjectPlaced(FortHeldObject)
  (bind (_returnvalue_roll _returnvalue_pitch _returnvalue_yaw) (Math|Rotator|BreakRotator (Transformation|GetActorRotation (HeldObjects|Instigator|GetInstigator (Variables|Default|GetFortHeldObject)))))
  ;[1 1 1] Set variables to track player interaction; report quest progress
  (CallFunction|ClientSetShouldTick false)
  ;[1 1 1] When we set the sign down, if it is overlapping with a player, push the sign away from the direction the player is facing
  (Utilities|Array|ForEachLoopwithBreak (Collision|GetOverlappingActors (Variables|Default|GetCapsule) "/Script/FortniteGame.FortPlayerPawn" :!class /Script/Engine.PrimitiveComponent)
    (:LoopBody
      (Variables|/Script/Engine.MovementComponent|SetVelocity (* (+ (* (Math|Vector|GetForwardVector (Transformation|GetActorRotation (HeldObjects|Instigator|GetInstigator (Variables|Default|GetFortHeldObject)))) -1.0) "0, 0,6.000000") 2e+02) (Variables|Default|GetFortProjectileMovement)))
    (:Completed
      _))
  (Variables|Default|SetHeld false)
  (Transformation|SetActorRotation 0 (Math|Rotator|MakeRotator 0.0 0.0 (+ _returnvalue_yaw 9e+01)) false)
  (CallFunction|ControlIcons true))

(event OnHeldObjectDropped(FortHeldObject)
  (bind (_returnvalue_roll _returnvalue_pitch _returnvalue_yaw) (Math|Rotator|BreakRotator (Transformation|GetActorRotation (HeldObjects|Instigator|GetInstigator (Variables|Default|GetFortHeldObject)))))
  ;[1 1 1] Set variables to track player interaction; report quest progress
  (CallFunction|ClientSetShouldTick false)
  (Variables|Default|SetHeld false)
  (Transformation|SetActorRotation 0 (Math|Rotator|MakeRotator 0.0 0.0 (+ _returnvalue_yaw 9e+01)) false)
  (CallFunction|ControlIcons true))

(event OnHeldObjectThrown(FortHeldObject)
  (bind (_returnvalue_roll _returnvalue_pitch _returnvalue_yaw) (Math|Rotator|BreakRotator (Transformation|GetActorRotation (HeldObjects|Instigator|GetInstigator (Variables|Default|GetFortHeldObject)))))
  ;[1 1 1] Set variables to track player interaction; report quest progress
  (CallFunction|ClientSetShouldTick false)
  (Variables|Default|SetHeld false)
  (Transformation|SetActorRotation 0 (Math|Rotator|MakeRotator 0.0 0.0 (+ _returnvalue_yaw 9e+01)) false)
  (CallFunction|ControlIcons true))

(event EventBeginPlay
  ;[1 1 1] Begin Play
  (|CallParentFunction)
  (Utilities|FlowControl|DelayUntilNextTick)
  (bind _asfort_player_pawn (Utilities|Casting|CastToFortPlayerPawn (Actor|GetOwner)))
  (Creative|GameplayEffects|SimpleApplyGameplayEffectToActor "/BattlePassS23/Items/QuestAbilities/Narrative/Part03/GE_Quest_Part03_CarryPlug.GE_Quest_Part03_CarryPlug_C" _asfort_player_pawn)
  (Variables|/Script/Engine.MovementComponent|SetVelocity "0, 0, 0" (Variables|Default|GetFortProjectileMovement))
  (if (not (Networking|IsServer))
    (Ability|SendGameplayEventtoActor _asfort_player_pawn "(TagName=\"Event.Narrative.Powerplug.Shown\")" (Utilities|Struct|MakeGameplayEventData "(TagName=\"\")" _asfort_player_pawn self))
    (CallFunction|ServerStartPickup)))

(event Custom|ServerStartPickup
  (bind _fortheldobject (Variables|Default|GetFortHeldObject))
  ;[1 1 1] Server Start Pickup
  (Utilities|FlowControl|Delay 0.2)
  (bind _asfort_player_pawn (Utilities|Casting|CastToFortPlayerPawn (Actor|GetOwner)))
  (HeldObjects|Behavior|PickupHeldObject _fortheldobject _asfort_player_pawn)
  ;[1 1 1] Bind to object attached, hide icons and start poi overlap checks
  (EventDispatchers|BindEventtoOnHeldObjectAttached _fortheldobject (delegate Custom|Rotate%20Sign))
  (CallFunction|ClientSetShouldTick true))

(event Custom|Rotate%20Sign
  ;[1 1 1] Bind to object attached, hide icons and start poi overlap checks
  (CallFunction|MulticastSetLocation)
  (CallFunction|ControlIcons false)
  (Variables|Default|SetHeld true)
  (EventDispatchers|UnbindEventfromOnHeldObjectAttached (Variables|Default|GetFortHeldObject) (delegate Custom|Rotate%20Sign)))

(event Custom|MulticastSetLocation
  ;[1 1 1] Sets Object Location while carried
  (Utilities|FlowControl|Delay 0.2)
  (Transformation|SetActorRelativeRotation (Variables|Default|GetDesiredRotation))
  (Transformation|SetActorRelativeLocation (Variables|Default|GetDesiredLocation))
  (Networking|ForceNetUpdate))

(event Interaction|EventBlueprintOnInteract
  (params (InteractingPawn :type Fort%20Pawn%20Object%20Reference) (InteractionBeingAttempted :type EInteractionBeingAttempted%20Enum :default "FirstInteraction"))
  ;[1 1 1] On Interact
  (switch Utilities|FlowControl|Switch|SwitchonEInteractionBeingAttempted InteractionBeingAttempted
    (:FirstInteraction
      _)
    (:SecondInteraction
      _)
    (:AllInteraction
      _))
  (bind _asfort_player_pawn (Utilities|Casting|CastToFortPlayerPawn InteractingPawn))
  (Variables|/Script/Engine.MovementComponent|SetVelocity "0, 0, 0" (Variables|Default|GetFortProjectileMovement))
  (if (Networking|IsServer)
    (HeldObjects|Behavior|PickupHeldObject (Variables|Default|GetFortHeldObject) _asfort_player_pawn)
    ;[1 1 1] Bind to object attached, hide icons and start poi overlap checks
    (EventDispatchers|BindEventtoOnHeldObjectAttached (Variables|Default|GetFortHeldObject) (delegate Custom|Rotate%20Sign))
    (CallFunction|ClientSetShouldTick true)))

(event Custom|ClearTimerAfterDelay
  ;[1 1 1] Stops the poi checks after a delay.
  (Utilities|FlowControl|Delay (Variables|Default|GetThrownTimerClearDelay))
  (if (not (Variables|Default|GetHeld))
    (Utilities|Time|ClearTimerbyFunctionName 0 "POI Overlap Check")))

(event OnLinkedActorDisconnected(FortLinkToActor)
  (if (not (Variables|Default|GetHeld))
    (Variables|ProjectileMovement|SetRepStartMoving 1)))

(event Custom|Control%20Icons
  (params (ShowIcons :type Boolean))
  ;[1 1 1] Call to show or hide icons, executed on client
  (CallFunction|OnRepShowIcon ShowIcons))

(event Custom|ClientSetShouldTick
  (params (ShouldReposition :type Boolean))
  (if (Networking|IsServer)
    _
    (elif ShouldReposition
      (Utilities|Time|SetTimerbyFunctionName 0 "LocalUpdateLocation" 0.001 true)
      (else
        (Utilities|Time|ClearTimerbyFunctionName 0 "LocalUpdateLocation")))))

(event Custom|ClientEnd
  (if (not (Networking|IsServer))
    (Components|Activation|Activate (Variables|Default|GetDeathExplosion) false)
    (Development|SetHiddeninGame (Variables|Default|GetStaticMesh) true false)))

(event Custom|DestroyWhenComplete
  (CallFunction|ControlIcons false)
  (Actor|DestroyActor self))

(event OnComponentHit(Capsule)
  (params (HitComponent :type Primitive%20Component%20Object%20Reference) (OtherActor :type Actor%20Object%20Reference) (OtherComp :type Primitive%20Component%20Object%20Reference) (NormalImpulse :type Vector :default "0, 0, 0") (Hit :type Hit%20Result%20Structure%20%28by%20ref%29))
  ;[1 1 1] Play sound que on hit
  (Utilities|Casting|CastToFortPawn OtherActor
    (:CastFailed
      (Audio|PlaySoundatLocation "/BattlePassS23/Audio/SFX/Part3/S23_Pt3_Q01_PlugDrop_Cue.S23_Pt3_Q01_PlugDrop_Cue" (Transformation|GetActorLocation)))))

(fn BlueprintCanInteract
  (params (InteractingPawn :type Fort%20Pawn%20Object%20Reference) (InteractionBeingAttempted :type EInteractionBeingAttempted%20Enum :default "FirstInteraction") (InteractionType :type TInteractionType%20Enum :default "IT_NoInteraction"))
  (Utilities|IsValid InteractingPawn
    (:Is%20Valid
      (bind _returnvalue (HeldObjects|Interaction|CanInteract (Variables|Default|GetFortHeldObject) 0))
      (bind _asfort_player_pawn (Utilities|Casting|CastToFortPlayerPawn (Actor|GetOwner)))
      (return (Math|Boolean|ANDBoolean (== (Utilities|Casting|CastToFortPlayerPawn InteractingPawn) _asfort_player_pawn) _returnvalue (not (GameplayTags|HasMatchingGameplayTag InteractingPawn "(TagName=\"Gameplay.Status.HidingInProp\")")) (not (GameplayTags|HasAnyMatchingGameplayTags InteractingPawn (Variables|Interact|GetTC_BlockPickup))))))))

(fn BlueprintGetInteractionString
  (params (InteractingPawn :type Fort%20Pawn%20Object%20Reference) (InteractionBeingAttempted :type EInteractionBeingAttempted%20Enum :default "FirstInteraction"))
  (return (Variables|Default|GetInteractString)))

(fn LocalUpdateLocation
  (Transformation|SetActorRelativeRotation (Variables|Default|GetDesiredRotation))
  (Transformation|SetActorRelativeLocation (Variables|Default|GetDesiredLocation)))

(fn OnRep_bIsInPOI
  (if (not (Variables|Default|GetbIsInPOI))
    (CallFunction|ClientEnd)))

(fn OnRep_Held
  (if (Variables|Default|GetHeld)
    (Utilities|Time|SetTimerbyFunctionName 0 "POI Overlap Check" 0.1 true)
    (if (not (Networking|IsServer))
      (bind _asfort_player_pawn (Utilities|Casting|CastToFortPlayerPawn (Actor|GetOwner)))
      (Ability|SendGameplayEventtoActor _asfort_player_pawn "(TagName=\"Event.Narrative.Powerplug.PickedUp\")" (Utilities|Struct|MakeGameplayEventData "(TagName=\"\")" _asfort_player_pawn)))
    (else
      (CallFunction|ClearTimerAfterDelay)
      (if (Networking|IsServer)
        _
        (elif (not (Variables|Default|GetTurnedIn))
          (bind _asfort_player_pawn_2 (Utilities|Casting|CastToFortPlayerPawn (Actor|GetOwner)))
          (Ability|SendGameplayEventtoActor _asfort_player_pawn_2 "(TagName=\"Event.Narrative.Powerplug.Dropped\")" (Utilities|Struct|MakeGameplayEventData "(TagName=\"\")" _asfort_player_pawn_2))
          (else
            (bind _asfort_player_pawn_3 (Utilities|Casting|CastToFortPlayerPawn (Actor|GetOwner)))
            (Ability|SendGameplayEventtoActor _asfort_player_pawn_3 "(TagName=\"Event.Narrative.Powerplug.TurnedIn\")" (Utilities|Struct|MakeGameplayEventData "(TagName=\"\")" _asfort_player_pawn_3))))))))

(fn OnRep_ShowIcon
  (params (ShowIcons :type Boolean))
  (bind _fortquesticon (Variables|Default|GetFortQuestIcon))
  (MiniMapData|SetViewDistanceRadius _fortquesticon (MiniMapData|GetViewDistanceRadius _fortquesticon))
  (MiniMapData|SetVisibility _fortquesticon ShowIcons)
  (Variables|Default|SetOwnerCanInteract ShowIcons))

(fn POIOverlapCheck
  (bind (_outactors _returnvalue) (Collision|CapsuleOverlapActors (Transformation|GetActorLocation) 65.0 65.0 (Utilities|Array|MakeArray "ObjectTypeQuery1" "ObjectTypeQuery2") "/BattlePassS23/QuestActors/Narrative/Part03/Quest_01/BP_S23_Socket.BP_S23_Socket_C"))
  (if _returnvalue
    (bind _asfort_player_pawn (Utilities|Casting|CastToFortPlayerPawn (Actor|GetOwner)))
    (bind _asfort_player_controller (Utilities|Casting|CastToFortPlayerController (Pawn|GetController _asfort_player_pawn)))
    (if (Networking|IsServer)
      (Quest|SendComplexCustomStatEvent (Quests|GetQuestManager _asfort_player_controller "Athena") 0 "(GameplayTags=)" "(GameplayTags=((TagName=\"Athena.Quest.S23.Part3.Power.PluggedIn\")))" 1)
      (else
        (bind _asfort_player_pawn_2 (Utilities|Casting|CastToFortPlayerPawn (Actor|GetOwner)))
        (Ability|SendGameplayEventtoActor _asfort_player_pawn_2 "(TagName=\"Event.Narrative.Powerplug.TurnedIn\")" (Utilities|Struct|MakeGameplayEventData "(TagName=\"\")" _asfort_player_pawn_2 self))))
    (Variables|Default|SetTurnedIn true)
    (HeldObjects|Behavior|DropHeldObject (Variables|Default|GetFortHeldObject))
    (Utilities|Time|ClearTimerbyFunctionName 0 "POI Overlap Check")
    (CallFunction|ControlIcons false)
    (Development|SetHiddeninGame (Variables|Default|GetStaticMesh) true false)))

(fn SetAttachDefaultTransform)

(fn ShouldSignBeEnabled
  (bind _outxy (DataTable|EvaluateCurveTableRow "/Game/Athena/Balance/DataTables/AthenaGameData.AthenaGameData" "Default.QuestInteractable.GiantPinkBear.Enabled")
    (:RowFound
      (if (>= _outxy 1.0)
        (return true)
        (else
          (return false))))
    (:RowNotFound
      (return false))))

(fn ConstructionScript
  (|CallParentFunction)
  (Effects|Components|ParticleSystem|SetTemplate (Variables|Default|GetDeathExplosion) (Variables|Default|GetBearPopEffect)))
