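; Quest pickup prop for the Season 23 narrative Part 3 Quest 1 power plug, which gives an interacting player a plug to carry.
; Interaction spawns a carry held object nearby, adjusts its capsule collision and maps it to the pawn.
; The locally controlled interactor receives a Hidden gameplay event and loses the quest icon.
; Quest readiness sends a delayed Shown event, and objective completion destroys that pawn's mapped carry actor.
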(event EventBeginPlay
  (|CallParentFunction) ; This node is disabled and will not be called.\nDrag off pins to build functionality.
)

(event Collision|EventActorBeginOverlap
  (params (OtherActor :type Actor%20Object%20Reference))
  (|CallParentFunction OtherActor) ; This node is disabled and will not be called.\nDrag off pins to build functionality.
)

(event EventTick
  (params (DeltaSeconds :type Float%20%28single-precision%29))
  (|CallParentFunction DeltaSeconds) ; This node is disabled and will not be called.\nDrag off pins to build functionality.
)

(event Interaction|EventBlueprintOnInteract
  (params (InteractingPawn :type Fort%20Pawn%20Object%20Reference) (InteractionBeingAttempted :type EInteractionBeingAttempted%20Enum :default "FirstInteraction"))
  (bind (_returnvalue_x _returnvalue_y) (Math|Vector|BreakVector (Transformation|GetActorLocation)))
  (bind (_returnvalue_roll _returnvalue_pitch _returnvalue_yaw) (Math|Rotator|BreakRotator (Transformation|GetActorRotation)))
  ;[1 1 1] Spawn Sign in stand at base - add player and spawned actor to map for removal later
  (CallFunction|ClientSendGameplayEventHidden InteractingPawn)
  (bind _returnvalue (Game|SpawnActorfromClass "/BattlePassS23/QuestActors/Narrative/Part03/Quest_01/BGA_HeldObject_Quest_PowerPlug_Carry.BGA_HeldObject_Quest_PowerPlug_Carry_C" (Math|Transform|MakeTransform (Math|Vector|MakeVector (- _returnvalue_x -7.0) (- _returnvalue_y -14.0)) (Math|Rotator|MakeRotator (- _returnvalue_roll 9e+01) _returnvalue_pitch (- _returnvalue_yaw 7.0))) "AlwaysSpawn" "OverrideRootScale" InteractingPawn 0 0))
  (CallFunction|SetCarryActorColision _returnvalue)
  (Utilities|Map|Add (Variables|Default|GetLinkedActor) InteractingPawn _returnvalue)
  (CallFunction|ReplicatedHideIcon InteractingPawn) ; Ideally we send this to only the instigating client but there isn't an easy way to do that
)

(event Custom|Replicated_HideIcon
  (params (InteractingPawn :type Fort%20Pawn%20Object%20Reference))
  ;[1 1 1] Set visibility of the icon to false on the client that triggered the interaction
  (if (Pawn|IsLocallyControlled InteractingPawn)
    (MiniMapData|SetVisibility (Variables|BuildingGameplayActorPropQuest|GetQuestIconComponent) false)))

(event Custom|ReplicatedShowIcons
  (params (InteractingPawn :type Fort%20Pawn%20Object%20Reference))
  ;[1 1 1] Set visibility of the icon to false on the client that triggered the interaction
  (if (Pawn|IsLocallyControlled InteractingPawn)
    (MiniMapData|SetVisibility (Variables|BuildingGameplayActorPropQuest|GetQuestIconComponent) true)))

(event Custom|ClientSendGameplayEventHidden
  (params (InteractingPawn :type Fort%20Pawn%20Object%20Reference))
  (if (Pawn|IsLocallyControlled InteractingPawn)
    (Ability|SendGameplayEventtoActor InteractingPawn "(TagName=\"Event.Narrative.Powerplug.Hidden\")" (Utilities|Struct|MakeGameplayEventData "(TagName=\"\")" InteractingPawn self))))

(event Custom|SetCarryActorColision
  (params (SpawnedActor :type BGA%20Held%20Object%20Quest%20Power%20Plug%20Carry%20Object%20Reference))
  (bind _capsule (Variables|/Game/Athena/Items/HeldObjects/HeldObjectVersion2/BGA_HeldObject_Parent.BGA_HeldObject_Parent_C|GetCapsule SpawnedActor))
  (Collision|SetCollisionResponsetoAllChannels _capsule "ECR_Overlap")
  (Collision|SetCollisionResponsetoChannel _capsule "ECC_WorldStatic" "ECR_Block"))

(event OnQuestReady(ConsolidatedQuestComponent)
  (params (PlayerController :type Fort%20Player%20Controller%20Object%20Reference))
  (Utilities|FlowControl|Delay 1.0)
  (Ability|SendGameplayEventtoActor (Game|Player|GetPlayerPawn PlayerController) "(TagName=\"Event.Narrative.Powerplug.Shown\")" (Utilities|Struct|MakeGameplayEventData "(TagName=\"\")" (Game|Player|GetPlayerPawn PlayerController) self)))

(event OnCustomUpdateObjective(ConsolidatedQuestComponent)
  (params (PlayerController :type Fort%20Player%20Controller%20Object%20Reference) (AchievedCount :type Integer) (RequiredCount :type Integer))
  (bind _returnvalue (Game|Player|GetPlayerPawn PlayerController))
  (bind _linkedactor (Variables|Default|GetLinkedActor))
  (bind (_value _returnvalue_2) (Utilities|Map|Find _linkedactor _returnvalue))
  ;[1 1 1] Destroy spawned actors for a player after their quest is completed
  (if (>= AchievedCount RequiredCount)
    (if (and (Utilities|IsValid _value :!class /Script/Engine.KismetSystemLibrary) _returnvalue_2)
      (Actor|DestroyActor _value))))

(fn ConstructionScript
  (|CallParentFunction))
